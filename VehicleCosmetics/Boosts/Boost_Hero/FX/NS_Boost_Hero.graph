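# Niagara system for the Hero vehicle boost cosmetic, drawing an icy blue exhaust trail of fog, sparks and ribbons.
# GPU emitter Fog001 blends three spawn groups into three sprite renderers using separate fog and spark attributes, while Frozen_Trail001 spawns quality-scaled ribbons per port.
# Both offset from DTCarOffset plus SingleBoostPos, inherit source movement and switch speed and size via bIsFrontendPreview.
# Painted-variant overrides, and the fog's skeletal socket placement, are disabled.
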NiagaraSystem /VehicleCosmetics/Boosts/Boost_Hero/FX/NS_Boost_Hero
dependencies {
 usedDataInterfaces [
  &NiagaraDataInterfaceCurve
  &NiagaraDataInterfaceSpriteRendererInfo
  &NiagaraDataInterfaceColorCurve
  &NiagaraDataInterfaceVector2DCurve
  &NiagaraDataInterfaceSkeletalMesh
 ]
 usedDynamicInputs []
 usedModules [
  &self:SystemSpawnScript.SystemScriptSource.SystemScriptGraph.NiagaraNodeAssignment_0.SetVariables_#1
  &SystemState
 ]
}
emitters {
 Fog001 {
  simTarget=GPUComputeSim
  emitterUpdateScript [
   &EmitterState()
   &VC_FilteredSocketCount()
   &SpawnRate(
    SpawnRate=&MakeCustomFloatFromBool(Bool=$User.bIsFrontendPreview, True Float=&RandomRangeFloat(Minimum=100, Maximum=150), False Float=100)
    Use Spawn Probability=false
   )
   &SpawnRate(SpawnGroup=1, SpawnRate=&RandomRangeFloat(Minimum=50, Maximum=30), Use Spawn Probability=false) {module=SpawnRate001}
   &SpawnRate(SpawnGroup=2, SpawnRate=60, Use Spawn Probability=false) {module=SpawnRate002}
  ]
  particleSpawnScript [
   &VC_PaintedVariantsOverride() {enabled=false}
   /SetVariables(Particles.Tags_Emitter:NiagaraInt32=&SpawnGroup()) {isSetParameters=true}
   &InitializeParticle(
    Color Mode=::NewEnumerator0
    Lifetime=&MakeCustomFloatFromBool(Bool=$User.bIsFrontendPreview, True Float=1, False Float=0.3)
    Position Offset=&AddVectorToPosition(
     Position=&MakeCustomVectorFromBool(Bool=$User.bIsFrontendPreview, True Vector=<0 0 0>, False Vector=<0 0 0>)
     Vector=&AddVectorToPosition(Position=$User.DTCarOffset, Vector=Linked() {linkedVariable {name=User.SingleBoostPos, type=NiagaraPosition}})
    )
    Position Offset Coordinate Space=::Local
    Sprite Rotation Angle=&RandomRangeFloat(Minimum=-1, Maximum=1)
    Sprite Rotation Mode=::NewEnumerator1
    Sprite Size Mode=::NewEnumerator1
    Uniform Sprite Size Max=&MakeCustomFloatFromBool(Bool=$User.bIsFrontendPreview, True Float=5, False Float=8)
    Uniform Sprite Size Min=&MakeCustomFloatFromBool(Bool=$User.bIsFrontendPreview, True Float=17.5, False Float=28)
    UsePositionOffset=true
    Write Material Random=true
   )
   &PartitionParticles(Number Of Partitions=$Emitter.PortCount)
   &SkeletalMeshLocation(
    Bone / Socket Index=$Particles.PartitionParticles.Partition
    Bone / Socket Sampling Mode=::NewEnumerator5
    MaskSpawnGroup=false
    Mesh Sampling Type=::NewEnumerator6
    Skeletal Mesh=$System.SkeletalMesh
   ) {enabled=false}
   &AddVelocity(Velocity=&MakeCustomVectorFromBool(
     Bool=$User.bIsFrontendPreview
     True Vector=&RandomRangeVector(Minimum=&MaskVectorBySpawnGroup(Masked Vector=<-450 0 0>, Passthrough Vector=<1 1 1>), Maximum=<-350 0 0>)
     False Vector=&RandomRangeVector(Minimum=<-2000 0 0>, Maximum=<-1800 0 0>)
    ))
   &AddVelocity(Velocity=&MaskVectorBySpawnGroup(
     Masked Vector=&RandomRangeVector(Minimum=<-500 100 100>, Maximum=<-400 -100 -100>)
     Passthrough Vector=<0 0 0>
     Spawn Group Mask=1
    )) {module=AddVelocity001}
   &AddVelocity(Velocity=&MaskVectorBySpawnGroup(
     Masked Vector=&RandomRangeVector(
      Minimum=&MakeCustomVectorFromBool(Bool=$User.bIsFrontendPreview, True Vector=<-1000 0 1>, False Vector=<-2000 0 0>)
      Maximum=&MakeCustomVectorFromBool(Bool=$User.bIsFrontendPreview, True Vector=<-800 0 1>, False Vector=<-1800 0 0>)
     )
     Passthrough Vector=<0 0 0>
     Spawn Group Mask=2
    )) {module=AddVelocity002}
   &ShapeLocation(
    MaskSpawnGroup=false
    Override Seed=false
    Sphere Radius=&MaskFloatBySpawnGroup(Masked Float=10, Passthrough Float=1)
    UseOwnerScale=true
   )
   &ShapeLocation(
    MaskSpawnGroup=false
    Override Seed=false
    Sphere Radius=&MaskFloatBySpawnGroup(Masked Float=&RandomRangeFloat(Minimum=20, Maximum=45), Passthrough Float=1, Spawn Group Mask=1)
    UseOwnerScale=true
   ) {module=ShapeLocation001}
   &ShapeLocation(
    MaskSpawnGroup=false
    Override Seed=false
    Sphere Radius=&MaskFloatBySpawnGroup(Masked Float=&RandomRangeFloat(Minimum=15, Maximum=25), Passthrough Float=1, Spawn Group Mask=2)
    UseOwnerScale=true
   ) {module=ShapeLocation002}
   &Color(Color=<rgba 0.4336 0.5945 1.5 2>)
   /SetVariables(
    Particles.Color_Fog:LinearColor=<rgba 0.4 0.6 1 1>
    Particles.Color_Sparks:LinearColor=<rgba 0.4 0.6 1 1>
    Particles.SpriteRotation_Fog:NiagaraFloat=&RandomRangeFloat(Minimum=0, Maximum=360)
    Particles.SpriteRotation_Sparks:NiagaraFloat=&RandomRangeFloat(Minimum=0, Maximum=1)
    Particles.SpriteSize_Fog:Vector2f=&RandomRangeVector2D(
     Minimum=&MakeCustomVector2DFromBool(Bool=$User.bIsFrontendPreview, True Vector2D=<120 120>, False Vector2D=<180 180>)
     Maximum=&MakeCustomVector2DFromBool(Bool=$User.bIsFrontendPreview, True Vector2D=<105 105>, False Vector2D=<210 210>)
    )
    Particles.SpriteSize_Sparks:Vector2f=&Multiply_Vector2D(
     A=&Multiply_Vector2D_ByFloat(Float=20, Vector2D=<0.1 0.5>)
     B=&Multiply_Vector2D_ByFloat(Float=20, Vector2D=<0.05 0.1>)
    )
    Particles.SubImageIndex_Fog:NiagaraFloat=$StackContext.SubImageIndex
    Particles.SubImageIndex_Sparks:NiagaraFloat=$StackContext.SubImageIndex
   ) {isSetParameters=true}
  ]
  particleUpdateScript [
   &ParticleState()
   &InheritSourceMovement(
    Applied Position Falloff Distance=0
    Applied Position Scale=&Multiply_VectorByFloat(
     Vector=<1 1 1>
     Float=&FloatFromCurve(FloatCurve=DataInterface() {propertyValues {Curve {keys [
          {interpMode=RCIM_Cubic, leaveTangent=0.0162, tangentMode=RCTM_Break, time=0, value=1}
          {arriveTangent=-0.0145, interpMode=RCIM_Cubic, tangentMode=RCTM_Break, time=1, value=1}
         ]}}})
    )
    ApplyForce=false
    ApplyPosition=true
    ApplyPositionFalloff=false
   )
   &ScaleColor(
    Scale Alpha=&FloatFromCurve(FloatCurve=DataInterface() {propertyValues {Curve {keys [
         {interpMode=RCIM_Cubic, tangentMode=RCTM_Break, time=0, value=0}
         {interpMode=RCIM_Cubic, tangentMode=RCTM_Break, time=0.1, value=0.07}
         {arriveTangent=-0.1529, interpMode=RCIM_Cubic, leaveTangent=-0.232, tangentMode=RCTM_Break, time=0.4, value=0.03}
         {interpMode=RCIM_Cubic, tangentMode=RCTM_Break, time=0.9, value=0}
        ]}}})
    Scale RGB=<3 3 3>
    ScaleA=true
    ScaleRGB=true
   )
   &ScaleSpriteSize(
    Uniform Curve Scale=1.5
    Uniform Curve Sprite Scale=DataInterface() {propertyValues {Curve {keys [
        {interpMode=RCIM_Cubic, leaveTangent=3.4385, tangentMode=RCTM_Break, time=0, value=0.5}
        {arriveTangent=8.9206, interpMode=RCIM_Cubic, leaveTangent=7.1313, tangentMode=RCTM_Break, time=0.4, value=4}
        {arriveTangent=0.9881, interpMode=RCIM_Cubic, leaveTangent=8.2248, tangentMode=RCTM_Break, time=0.8, value=6.75}
        {arriveTangent=8.0149, interpMode=RCIM_Cubic, tangentMode=RCTM_Break, time=0.97, value=7.49}
       ]}}}
   )
   &CurlNoiseForce(Bias Noise=false, Pan Noise=false)
   &SolveForcesAndVelocity(Clamp Velocity=false, Limit Acceleration=false)
   &SubUVAnimation(
    "Sprite Renderer "=DataInterface() {propertyValues {SpriteRenderer=&self:Fog001_0.NiagaraSpriteRendererProperties_3}}
    UseEndFrame=true
    UseStartFrame=true
   )
   /SetVariables(
    Particles.Color_Fog:LinearColor=&ColorFromCurve(ColorCurve=DataInterface() {propertyValues {
       AlphaCurve {keys [
         {interpMode=RCIM_Cubic, leaveTangentWeight=0.3333, tangentMode=RCTM_Break, time=0, value=0}
         {
          arriveTangentWeight=0.0667
          interpMode=RCIM_Cubic
          leaveTangent=0.0994
          leaveTangentWeight=0.2667
          tangentMode=RCTM_Break
          time=0.2
          value=0.8
         }
         {
          arriveTangent=-1.6525
          arriveTangentWeight=0.1
          interpMode=RCIM_Cubic
          leaveTangent=-1.5058
          leaveTangentWeight=0.1667
          tangentMode=RCTM_Break
          time=0.5
          value=0.1
         }
         {arriveTangent=-0.0448, arriveTangentWeight=0.1, interpMode=RCIM_Cubic, leaveTangentWeight=0.0667, tangentMode=RCTM_Break, time=0.6, value=0}
        ]}
       BlueCurve {keys [<at 0 1 RCIM_Constant>,<at 1 1 RCIM_Cubic>]}
       GreenCurve {keys [<at 0 0.6458 RCIM_Cubic>,<at 1 0.6424 RCIM_Cubic>]}
       RedCurve {keys [<at 0 0.4688 RCIM_Cubic>,<at 1 0.4635 RCIM_Cubic>]}
      }})
    Particles.Color_Sparks:LinearColor=&ColorFromCurve(ColorCurve=DataInterface() {propertyValues {
       AlphaCurve {keys [
         <at 0.1 0 RCIM_Cubic>
         <at 0.3 1 RCIM_Cubic>
         {arriveTangent=-2.6853, interpMode=RCIM_Cubic, leaveTangent=-2.6853, tangentMode=RCTM_User, time=0.3705, value=0}
        ]}
       BlueCurve {keys [<at 0 7 RCIM_Cubic>,<at 1 7 RCIM_Cubic>]}
       GreenCurve {keys [<at 0 7 RCIM_Cubic>,<at 1 7 RCIM_Cubic>]}
       RedCurve {keys [<at 0 7 RCIM_Cubic>,<at 1 7 RCIM_Cubic>]}
      }})
    Particles.SpriteSize_Fog:Vector2f=&Vector2DFromCurve(
     Vector2Curve=DataInterface() {propertyValues {
       XCurve {keys [
         {interpMode=RCIM_Cubic, tangentMode=RCTM_Break, time=0.03, value=0.33}
         {interpMode=RCIM_Cubic, tangentMode=RCTM_Break, time=0.5, value=1}
         <at 1 1>
        ]}
       YCurve {keys [{interpMode=RCIM_Cubic, tangentMode=RCTM_Break, time=0.03, value=0.33}, <at 0.5 1>, <at 1 1>]}
      }}
     ScaleCurve=<150 150>
    )
   ) {isSetParameters=true}
   &DynamicMaterialParameters(
    Index 0 Param 1=0.2
    Index 0 Param 2=&RandomRangeFloat(Minimum=3, Maximum=2)
    Module.Index0Float0Write=true
    Module.Index0Float1Write=true
    Module.Index0Float2Write=true
    Module.Index0Float3Write=true
   )
  ]
  rendererClasses [&NiagaraSpriteRendererProperties]
  renderers [
   {rendererClass=&NiagaraSpriteRendererProperties}
   {rendererClass=&NiagaraSpriteRendererProperties, rendererIndex=1}
   {rendererClass=&NiagaraSpriteRendererProperties, rendererIndex=2}
  ]
 }
 Frozen_Trail001 {
  emitterUpdateScript [
   &EmitterState(Life Cycle Mode=::NewEnumerator1, Loop Duration=1, UseLoopDelay=false)
   &VC_FilteredSocketCount()
   &VC_RateQualityHandler(Epic Spawn Rate=30, High Spawn Rate=20, Medium Spawn Rate=18, Use Custom Rate?=true)
   &VC_SpawnRibbon_v1(Number of Ribbons=$Emitter.PortCount, SpawnRate=$Emitter.Spawn Rate)
  ]
  particleSpawnScript [
   &VC_PaintedVariantsOverride() {enabled=false}
   &VC_InitializeRibbon_v1(
    Custom Offset=true
    Lifetime=&Multiply_Float(A=1.5, B=1)
    Local Vector Offset=&AddVectorToPosition(
     Position=&AddVectorToPosition(Position=$User.DTCarOffset, Vector=Linked() {linkedVariable {name=User.SingleBoostPos, type=NiagaraPosition}})
     Vector=&MakeCustomVectorFromBool(
      Bool=$User.bIsFrontendPreview
      True Vector=HlslExpression() {hlslExpression="Engine.QualityLevel == 0 ? float3(-0,0,0) : float3(10,0,0)"}
      False Vector=HlslExpression() {hlslExpression="Engine.QualityLevel == 0 ? float3(0,0,0) : float3(0,0,0)"}
     )
    )
    Offset Amount=1
    Ribbon Width:NiagaraFloat=&MakeCustomFloatFromBool(
     Bool=$User.bIsFrontendPreview
     True Float=&Multiply_Float(A=30, B=3)
     False Float=&Multiply_Float(A=30, B=5)
    )
   )
   &AddVelocity(Velocity=&MakeCustomVectorFromBool(Bool=$User.bIsFrontendPreview, True Vector=<-1000 0 0>, False Vector=<-2000 0 0>)) {module=AddVelocity001}
   &Color(Color=<rgba 0.377 0.4403 1.4688 2>) {module=Color001}
  ]
  particleUpdateScript [
   &ParticleState()
   &InheritSourceMovement(
    Applied Position Scale=&Multiply_VectorByFloat(
     Vector=<1 1 1>
     Float=&FloatFromCurve(FloatCurve=DataInterface() {propertyValues {Curve {keys [
          {interpMode=RCIM_Cubic, leaveTangent=-0.0051, tangentMode=RCTM_Break, time=0, value=1}
          {arriveTangent=-0.247, interpMode=RCIM_Cubic, tangentMode=RCTM_Break, time=1, value=0.9}
         ]}}})
    )
    ApplyForce=false
    ApplyPosition=true
    ApplyPositionFalloff=false
   )
   &SolveForcesAndVelocity(Clamp Velocity=false, Limit Acceleration=false)
   &VC_UpdateRibbon_v1(Enable Skeletal Position=true)
   &ScaleColor(
    Scale Alpha=&FloatFromCurve(
     FloatCurve=DataInterface() {propertyValues {Curve {keys [
         {interpMode=RCIM_Cubic, leaveTangent=0.0223, tangentMode=RCTM_Break, time=0, value=0}
         {arriveTangent=3.2359, interpMode=RCIM_Cubic, leaveTangent=3.2359, tangentMode=RCTM_User, time=0.1, value=0.1}
         {arriveTangent=1.1965, interpMode=RCIM_Cubic, leaveTangent=-0.2409, tangentMode=RCTM_Break, time=0.2, value=1}
         {interpMode=RCIM_Cubic, tangentMode=RCTM_Break, time=0.9, value=0}
        ]}}}
     Scale Curve=3
    )
    Scale RGB=<6 6 6>
    ScaleA=true
    ScaleRGB=true
   ) {module=ScaleColor001}
   &DynamicMaterialParameters(
    Index 0 Param 1=1
    Index 0 Param 2=3
    Index 0 Param 3=3
    Module.Index0Float0Write=true
    Module.Index0Float1Write=true
    Module.Index0Float2Write=true
    Module.Index0Float3Write=true
   )
   &ScaleRibbonWidth(Ribbon Width Scale=&FloatFromCurve(FloatCurve=DataInterface() {propertyValues {Curve {keys [
         {interpMode=RCIM_Cubic, leaveTangent=3.3375, tangentMode=RCTM_Break, time=0, value=0.2}
         {arriveTangent=1.6401, interpMode=RCIM_Cubic, leaveTangent=1.223, tangentMode=RCTM_Break, time=0.3, value=0.85}
         {arriveTangent=0.219, interpMode=RCIM_Cubic, tangentMode=RCTM_Break, time=1, value=1.5}
        ]}}}))
  ]
  rendererClasses [&NiagaraRibbonRendererProperties]
  renderers [{rendererClass=&NiagaraRibbonRendererProperties}]
 }
}
summary {
 systemName=NS_Boost_Hero
 emitters [
  {emitterName=Frozen_Trail001, rendererClasses [&NiagaraRibbonRendererProperties]}
  {emitterName=Fog001, simTarget=GPUComputeSim, rendererClasses [&NiagaraSpriteRendererProperties]}
 ]
 userVariables [
  User.bIsFrontendPreview {type=NiagaraBool, defaultValue=true}
  User.bIsPainted {type=NiagaraBool}
  User.ColorVariantID {type=NiagaraInt32}
  User.DTCarOffset {type=NiagaraPosition, description="", defaultValue {}}
  User.PaintedColor {type=LinearColor, description="", defaultValue {struct=LinearColor, value=<rgba 1 1 1 1>}}
  User.PaintedScalar {type=NiagaraFloat, defaultValue=1, description=""}
  User.PortCount {type=NiagaraInt32, defaultValue=1}
  User.SingleBoostPos {type=NiagaraPosition, description="", defaultValue {}}
 ]
}
